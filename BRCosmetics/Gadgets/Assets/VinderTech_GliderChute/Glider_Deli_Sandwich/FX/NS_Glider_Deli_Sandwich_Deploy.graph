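# Niagara effect for the Deli Sandwich glider deploy, built as a short burst from four emitters.
# DeliSandwish_VFX briefly spawns growing, fading meshes with slight random spin and animated dynamic material parameters.
# Sprite spikes fly outward from a sphere, while single shockwave and wind meshes expand and fade through age curves.
# The shockwave also writes animated light color and radius, although no emitter has a light renderer.
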NiagaraSystem /BRCosmetics/Gadgets/Assets/VinderTech_GliderChute/Glider_Deli_Sandwich/FX/NS_Glider_Deli_Sandwich_Deploy
dependencies {
 usedDataInterfaces [
  &NiagaraDataInterfaceCurve
  &NiagaraDataInterfaceVector2DCurve
  &NiagaraDataInterfaceVectorCurve
 ]
 usedDynamicInputs []
 usedModules [
  &self:SystemSpawnScript.SystemScriptSource.SystemScriptGraph.NiagaraNodeAssignment_0.SetVariables_#1
  &SystemState
 ]
}
emitters {
 DeliSandwish_VFX {
  emitterUpdateScript [
   &EmitterState(Life Cycle Mode=::NewEnumerator1, Loop Behavior=::NewEnumerator1, Loop Delay=0.6, Loop Duration=0.3, UseLoopDelay=false)
   &SpawnRate(SpawnRate=9, Use Spawn Probability=false)
   &SpawnBurst_Instantaneous(Spawn Count=1, Use Spawn Probability=false)
  ]
  particleSpawnScript [
   /Niagara/Modules/Spawn/Initialization/InitializeParticle(
    Color=<rgba 5 0.1245 1.6855 1>
    Lifetime=0.475
    Mass=&UniformRangedFloat(Minimum=0.6, Maximum=1.2)
    Mesh Scale=&VectorFromFloat(Value=5)
    Module.Write Color=true
    Module.Write Lifetime=true
    Module.Write Mass=true
    Module.Write Material Random=false
    Module.Write Position=false
    Module.Write Scale=true
    Module.Write SpriteRotation=false
    Module.Write SpriteSize=false
    Position=&SimulationPosition()
    Sprite Rotation=&NormalizedAngleToDegrees(NormalizedAngle=&UniformRangedFloat(Minimum=0, Maximum=1))
   )
   &AddRotationalVelocity(Rotation Rate=&UniformRangedVector(Minimum=<-0.1 0 0>, Maximum=<0.1 0 0>))
   &InitialMeshOrientation(
    Orientation Vector=&UniformRangedVector(Minimum=<-0.5 0 0>, Maximum=<0.5 0 0>)
    Use Orientation Vector=true
    Use Rotation Vector=false
   ) {module=InitialMeshOrientation001}
  ]
  particleUpdateScript [
   &SolveForcesAndVelocity(Clamp Velocity=false, Limit Acceleration=false)
   &SolveRotationalForcesAndVelocity()
   &ParticleState()
   &ScaleMeshSize(Scale Factor=&VectorFromFloat(Value=&FloatFromCurve(FloatCurve=DataInterface() {propertyValues {Curve {keys [{arriveTangent=1.6678, time=0, value=0.25}, {arriveTangent=0.8778, leaveTangent=0.8778, time=1, value=2.2}]}}})))
   &ScaleColor(
    Scale Alpha=&FloatFromCurve(FloatCurve=DataInterface() {propertyValues {Curve {keys [
         {arriveTangent=8.3351, interpMode=RCIM_Cubic, leaveTangent=8.3351, tangentMode=RCTM_User, time=0, value=0}
         {interpMode=RCIM_Cubic, tangentMode=RCTM_User, time=0.1397, value=1}
         <at 0.8116 0.0184>
         {interpMode=RCIM_Cubic, tangentMode=RCTM_User, time=1, value=0}
        ]}}})
    ScaleA=true
    ScaleRGB=false
   )
   &DynamicMaterialParameters(
    Index 0 Param 1=&FloatFromCurve(FloatCurve=DataInterface() {propertyValues {Curve {keys [{arriveTangent=3.2486, interpMode=RCIM_Cubic, leaveTangent=3.2485, tangentMode=RCTM_User, time=0, value=0}, <at 1 2>]}}})
    Index 0 Param 4=&FloatFromCurve(FloatCurve=DataInterface() {propertyValues {Curve {keys [<at 0 5>,<at 1 6>]}}}, Scale Curve=3.5)
    Module.Param0WriteEnabled=true
    Module.Param1WriteEnabled=false
    Module.Param2WriteEnabled=false
    Module.Param3WriteEnabled=true
   )
  ]
  rendererClasses [&NiagaraMeshRendererProperties]
  renderers [{rendererClass=&NiagaraMeshRendererProperties, rendererIndex=0}]
 }
 DirectionalBurstSpikes {
  emitterUpdateScript [
   &EmitterState()
   &SpawnBurst_Instantaneous(Spawn Count=&UniformRangedInt(Minimum=18, Maximum=24), Use Spawn Probability=false)
  ]
  particleSpawnScript [
   &InitializeParticle(
    Color=&UniformRangedLinearColor(Minimum=<rgba 2 0.8 1.1842 0.4>, Maximum=<rgba 2 0.8 1.1842 0.25>)
    Lifetime=&UniformRangedFloat(Minimum=0.15, Maximum=0.25)
    Mass=&UniformRangedFloat(Minimum=0.75, Maximum=1.75)
    Module.Write Color=true
    Module.Write Lifetime=true
    Module.Write Mass=true
    Module.Write Material Random=true
    Module.Write Position=true
    Module.Write Scale=false
    Module.Write SpriteRotation=false
    Module.Write SpriteSize=true
    Position=&SimulationPosition()
    Sprite Size=&UniformRangedVector2D(Minimum=<10 80>, Maximum=<50 150>)
   )
   /Niagara/Modules/Spawn/Location/SphereLocation(MaskSpawnGroup=false, Module.Override Randomness=false, Sphere Radius=60, Surface Only=false)
   &AddVelocityFromPoint(Module.Use Falloff Distance=false, Velocity Strength=500)
   &DynamicMaterialParameters(
    Index 0 Param 1=50
    Module.Param0WriteEnabled=true
    Module.Param1WriteEnabled=true
    Module.Param2WriteEnabled=true
    Module.Param3WriteEnabled=true
   )
  ]
  particleUpdateScript [
   &ParticleState()
   &ScaleColor(
    Scale Alpha=&FloatFromCurve(FloatCurve=DataInterface() {propertyValues {Curve {keys [
         {interpMode=RCIM_Cubic, leaveTangent=-0.0594, tangentMode=RCTM_Break, time=-0.0029, value=1.0303}
         {arriveTangent=-0.936, interpMode=RCIM_Cubic, leaveTangent=-0.936, tangentMode=RCTM_User, time=0.0887, value=0.3527}
         <at 1.0011 0 RCIM_Cubic>
        ]}}})
    ScaleA=true
    ScaleRGB=true
   )
   &SolveForcesAndVelocity(Clamp Velocity=false, Limit Acceleration=false)
   &ScaleSpriteSize(Scale Factor=&Vector2DFromCurve(
     Vector2Curve=DataInterface() {propertyValues {
       XCurve {keys [<at 0.0014 0.4615>,<at 1 1>]}
       YCurve {keys [{arriveTangent=1.6918, interpMode=RCIM_Cubic, leaveTangent=1.6918, tangentMode=RCTM_User, time=-0.0006, value=0.4523}, <at 1 1>]}
      }}
     ScaleCurve=<1 1>
    ))
  ]
  rendererClasses [&NiagaraSpriteRendererProperties]
  renderers [{rendererClass=&NiagaraSpriteRendererProperties}]
 }
 HT_LoopShockwave_CamF {
  emitterUpdateScript [
   &EmitterState()
   &SpawnBurst_Instantaneous(Spawn Count=1, Use Spawn Probability=false)
  ]
  particleSpawnScript [
   &InitializeParticle(
    Color=<rgba 3 1.2 1.7763 0.25>
    Lifetime=0.4
    Mesh Scale=&VectorFromFloat(Value=3)
    Module.Write Color=true
    Module.Write Lifetime=true
    Module.Write Mass=true
    Module.Write Material Random=true
    Module.Write Position=false
    Module.Write Scale=true
    Module.Write SpriteRotation=false
    Module.Write SpriteSize=false
    Position=&SimulationPosition()
   )
   &InitialMeshOrientation(
    Orientation Axis=<0 0 1>
    Orientation Vector=<0 1 0>
    Rotation=&MakeVector(X=&UniformRangedFloat(Minimum=0, Maximum=1), Y=0, Z=0)
    Use Orientation Vector=true
    Use Rotation Vector=true
   ) {module=InitialMeshOrientation001}
  ]
  particleUpdateScript [
   &ParticleState()
   &ScaleMeshSize(Scale Factor=/Niagara/DynamicInputs/ScaleByCurve/ScaleVectorByCurve(
     InitialValue:Vector3f=<1 1 1>
     ScaleCurve:NiagaraDataInterfaceVectorCurve=DataInterface() {propertyValues {
       XCurve {keys [
         {arriveTangent=7.5627, interpMode=RCIM_Cubic, leaveTangent=3.2485, tangentMode=RCTM_User, time=0, value=0}
         {interpMode=RCIM_Cubic, tangentMode=RCTM_User, time=1, value=1}
        ]}
       YCurve {keys [
         {arriveTangent=9.2714, interpMode=RCIM_Cubic, leaveTangent=3.2485, tangentMode=RCTM_User, time=0, value=0}
         {interpMode=RCIM_Cubic, tangentMode=RCTM_User, time=1, value=1}
        ]}
       ZCurve {keys [
         {arriveTangent=8.0047, interpMode=RCIM_Cubic, leaveTangent=3.2485, tangentMode=RCTM_User, time=0, value=0}
         {interpMode=RCIM_Cubic, tangentMode=RCTM_User, time=1, value=1}
        ]}
      }}
     CurveIndex:NiagaraFloat=$Particles.NormalizedAge
    ))
   &ScaleColor(
    Scale Alpha=&FloatFromCurve(FloatCurve=DataInterface() {propertyValues {Curve {keys [
         {interpMode=RCIM_Cubic, tangentMode=RCTM_User, time=0.0023, value=0}
         {arriveTangent=2.0824, interpMode=RCIM_Cubic, leaveTangent=0.0112, tangentMode=RCTM_Break, time=0.0552, value=0.9565}
         {arriveTangent=-0.1758, interpMode=RCIM_Cubic, leaveTangent=-0.1758, tangentMode=RCTM_User, time=1.0011, value=0.002}
        ]}}})
    ScaleA=true
    ScaleRGB=true
   )
   /SetVariables(Particles.LightColor=&MakeLinearColorFromVectorAndFloat(
     "Vector (RGB)"=&VectorFromCurve(VectorCurve=DataInterface() {propertyValues {
        XCurve {keys [
          {interpMode=RCIM_Cubic, tangentMode=RCTM_Break, time=-0.0014, value=9.2729}
          {arriveTangent=-15.7559, interpMode=RCIM_Cubic, leaveTangent=-0.3065, tangentMode=RCTM_Break, time=0.3017, value=-0.1656}
          <at 1 0>
         ]}
        YCurve {keys [
          {interpMode=RCIM_Cubic, leaveTangent=30.1692, tangentMode=RCTM_Break, time=0, value=0}
          {interpMode=RCIM_Cubic, leaveTangent=-34.8929, tangentMode=RCTM_Break, time=0.2237, value=12}
          {interpMode=RCIM_Cubic, tangentMode=RCTM_Break, time=1, value=0}
         ]}
        ZCurve {keys [<at 0 0>,<at 0.0627 0>,<at 1 0>]}
       }})
     "Float (Alpha)"=1
    )) {isSetParameters=true}
   /SetVariables(Particles.LightRadius=&FloatFromCurve(FloatCurve=DataInterface() {propertyValues {Curve {keys [<at 0.0021 0.0215>,<at 0.1455 0.1865>,<at 0.9014 1>]}}}, Scale Curve=125)) {isSetParameters=true}
   &DynamicMaterialParameters(
    Index 0 Param 1=&FloatFromCurve(FloatCurve=DataInterface() {propertyValues {Curve {keys [<at 0 2.5>,<at 1 0>]}}})
    Index 0 Param 2=&FloatFromCurve(FloatCurve=DataInterface() {propertyValues {Curve {keys [<at 0 3>,<at 1 0>]}}}, Scale Curve=5)
    Index 0 Param 3=1
    Module.Param0WriteEnabled=true
    Module.Param1WriteEnabled=true
    Module.Param2WriteEnabled=true
    Module.Param3WriteEnabled=true
   )
  ]
  rendererClasses [&NiagaraMeshRendererProperties]
  renderers [{rendererClass=&NiagaraMeshRendererProperties, rendererIndex=0}]
 }
 WindMeshBurt {
  emitterUpdateScript [
   &EmitterState()
   &SpawnBurst_Instantaneous(Spawn Count=1, Use Spawn Probability=false)
  ]
  particleSpawnScript [&InitializeParticle(
    Color=<rgba 1 1 1 0.75>
    Lifetime=0.225
    Mesh Scale=&VectorFromFloat(Value=2)
    Module.Write Color=true
    Module.Write Lifetime=true
    Module.Write Mass=true
    Module.Write Material Random=true
    Module.Write Position=true
    Module.Write Scale=true
    Module.Write SpriteRotation=false
    Module.Write SpriteSize=false
    Position=&SimulationPosition()
    Sprite Size=<25 25>
   )]
  particleUpdateScript [
   &ParticleState()
   &ScaleColor(
    Scale Alpha=&FloatFromCurve(FloatCurve=DataInterface() {propertyValues {Curve {keys [<at 0.2946 1>,<at 1 0>]}}})
    Scale RGB=&MakeVectorFromLinearColorRGB(Color=<rgba 1 1 1 1>)
    ScaleA=true
    ScaleRGB=true
   )
   &ScaleMeshSize(Scale Factor=&VectorFromFloat(Value=&FloatFromCurve(FloatCurve=DataInterface() {propertyValues {Curve {keys [{arriveTangent=2.5191, interpMode=RCIM_Cubic, leaveTangent=2.5191, tangentMode=RCTM_User, time=0, value=0}, <at 1 1>]}}})))
  ]
  rendererClasses [&NiagaraMeshRendererProperties]
  renderers [{rendererClass=&NiagaraMeshRendererProperties, rendererIndex=0}]
 }
}
summary {
 systemName=NS_Glider_Deli_Sandwich_Deploy
 emitters [
  {emitterName=WindMeshBurt, rendererClasses [&NiagaraMeshRendererProperties]}
  {emitterName=DirectionalBurstSpikes, rendererClasses [&NiagaraSpriteRendererProperties]}
  {emitterName=HT_LoopShockwave_CamF, rendererClasses [&NiagaraMeshRendererProperties]}
  {emitterName=DeliSandwish_VFX, rendererClasses [&NiagaraMeshRendererProperties]}
 ]
 userVariables [User.windowSize {type=Vector3f, description="", defaultValue {struct=Vector3f, value=<0 0 0>}}]
}
