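# Niagara effect for the Deli Sandwich glider, pinning looping pink visuals and wing-flap trail meshes to the glider's skeletal mesh.
# A mesh emitter spawns growing, fading particles at one socket, alongside two fixed-facing glow sprites on the same loop timing.
# Wing foils spawn only when fully deployed, using spawn groups to mirror meshes across two sockets, with alpha scaled by a system integrator.
# Particles resample socket transforms every update to follow the glider.
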NiagaraSystem /BRCosmetics/Gadgets/Assets/VinderTech_GliderChute/Glider_Deli_Sandwich/FX/NS_Glider_Deli_Sandwich_Trails_01
dependencies {
 usedDataInterfaces [&NiagaraDataInterfaceCurve, &NiagaraDataInterfaceColorCurve]
 usedDynamicInputs []
 usedModules [
  &SystemState
  /Game/Effects/Niagara/Modules/Math/FN_Integrator/FN_Integrator_v2
  &FN_Integrator
 ]
}
emitters {
 DeliSandwish_VFX {
  emitterUpdateScript [
   &EmitterState(Inactive Response=::NewEnumerator1, Life Cycle Mode=::NewEnumerator1, Loop Delay=0.6, Loop Duration=1.95, UseLoopDelay=true)
   &SpawnRate(SpawnRate=3, Use Spawn Probability=false)
  ]
  particleSpawnScript [
   /Niagara/Modules/Spawn/Initialization/InitializeParticle(
    Color=<rgba 5 0.1245 1.6855 1>
    Lifetime=0.95
    Mass=&UniformRangedFloat(Minimum=0.6, Maximum=1.2)
    Mesh Scale=&VectorFromFloat(Value=1.25)
    Module.Write Color=true
    Module.Write Lifetime=true
    Module.Write Mass=true
    Module.Write Material Random=false
    Module.Write Position=false
    Module.Write Scale=true
    Module.Write SpriteRotation=false
    Module.Write SpriteSize=false
    Position=&SimulationPosition()
    Sprite Rotation=&NormalizedAngleToDegrees(NormalizedAngle=&UniformRangedFloat(Minimum=0, Maximum=1))
   )
   &SystemLocation(MaskSpawnGroup=false, Offset=<25 0 10>) {enabled=false}
   &AddVelocity(Velocity=<40 0 0>) {enabled=false}
   &AddRotationalVelocity(Rotation Rate=&UniformRangedVector(Minimum=<-0.1 0 0>, Maximum=<0.1 0 0>))
   &InitialMeshOrientation(
    Orientation Vector=&UniformRangedVector(Minimum=<-0.5 0 0>, Maximum=<0.5 0 0>)
    Use Orientation Vector=true
    Use Rotation Vector=false
   ) {module=InitialMeshOrientation001}
  ]
  particleUpdateScript [
   &ParticleState()
   &SolveForcesAndVelocity(Clamp Velocity=false, Limit Acceleration=false)
   &SolveRotationalForcesAndVelocity()
   &SkeletalMeshLocation(
    Apply Sampled Orientation to Mesh Particles=true
    Bone / Socket Index=6
    Inherit Velocity=false
    MaskSpawnGroup=false
    Mesh Sampling Type=::NewEnumerator5
    Sampled Position Offset=<27 0 0>
    Sampled Position Offset Coordinate Space:ENiagaraMeshTransforms=::NewEnumerator3
    Skeletal Mesh=$User.SkeletalMesh
    Socket Sampling Mode=::NewEnumerator5
    Use Sampled Mesh Attribute as Ribbon ID=false
    Write Sampled Mesh Attrributes to Particle Attributes=true
   )
   /SetVariables(Particles.DistanceTraveledLocal:Vector3f=&Add_Vector(
     A=Linked() {linkedVariable {name=Particles.Position, type=NiagaraPosition}}
     B=&RotateVectorByQuaternion(
      Vector To Rotate=&Multiply_VectorByFloat(Vector=<43 0 0>, Float=$Particles.NormalizedAge)
      Quaternion=$Output.SkeletalMeshLocation.SampledRotation
     )
    )) {isSetParameters=true}
   /SetVariables(Particles.LightAlpha:NiagaraFloat=&FloatFromCurve(
     FloatCurve=DataInterface() {propertyValues {Curve {keys [
         <at 0 0>
         <at 0.1 1>
         {arriveTangent=-3.6605, leaveTangent=-3.6605, time=0.675, value=1}
         {arriveTangent=-3.1112, leaveTangent=-3.1112, time=1, value=0}
        ]}}}
     CurveIndex=$Emitter.NormalizedLoopAge
    )) {isSetParameters=true}
   &ScaleMeshSize(Scale Factor=&VectorFromFloat(Value=&FloatFromCurve(FloatCurve=DataInterface() {propertyValues {Curve {keys [
          {arriveTangent=0.1305, interpMode=RCIM_Cubic, leaveTangent=0.1305, tangentMode=RCTM_User, time=0, value=1}
          {arriveTangent=0.1729, interpMode=RCIM_Cubic, leaveTangent=0.173, tangentMode=RCTM_User, time=1, value=2.2}
         ]}}})))
   &ScaleColor(
    Scale Alpha=&FloatFromCurve(
     FloatCurve=DataInterface() {propertyValues {Curve {keys [
         {arriveTangent=8.3351, interpMode=RCIM_Cubic, leaveTangent=8.3351, tangentMode=RCTM_User, time=0, value=0}
         {interpMode=RCIM_Cubic, tangentMode=RCTM_User, time=0.1397, value=1}
         <at 0.8116 0.0184>
         {interpMode=RCIM_Cubic, tangentMode=RCTM_User, time=1, value=0}
        ]}}}
     Scale Curve=$Particles.LightAlpha
    )
    ScaleA=true
    ScaleRGB=false
   )
   &DynamicMaterialParameters(
    Index 0 Param 1=&FloatFromCurve(FloatCurve=DataInterface() {propertyValues {Curve {keys [{arriveTangent=3.2486, interpMode=RCIM_Cubic, leaveTangent=3.2485, tangentMode=RCTM_User, time=0, value=0}, <at 1 2>]}}})
    Index 0 Param 4=&FloatFromCurve(FloatCurve=DataInterface() {propertyValues {Curve {keys [<at 0 5>,<at 1 6>]}}}, Scale Curve=1.5)
    Module.Param0WriteEnabled=true
    Module.Param1WriteEnabled=false
    Module.Param2WriteEnabled=false
    Module.Param3WriteEnabled=true
   )
  ]
  rendererClasses [&NiagaraMeshRendererProperties]
  renderers [{rendererClass=&NiagaraMeshRendererProperties, rendererIndex=0}]
 }
 NE_Decal {
  emitterUpdateScript [
   &EmitterState(Inactive Response=::NewEnumerator1, Life Cycle Mode=::NewEnumerator1, Loop Delay=0.6, Loop Duration=1.95, UseLoopDelay=true)
   &SpawnBurst_Instantaneous(Spawn Count=1, Spawn Time=0.05, Use Spawn Probability=false)
  ]
  particleSpawnScript [
   &InitializeParticle(
    Lifetime=2
    Module.Write Color=true
    Module.Write Lifetime=true
    Module.Write Mass=true
    Module.Write Material Random=true
    Module.Write Position=false
    Module.Write Scale=false
    Module.Write SpriteRotation=false
    Module.Write SpriteSize=true
    Sprite Size=&Vector2DFromFloat(Value=85)
   )
   &SystemLocation(MaskSpawnGroup=false, Offset=<30 0 10>) {enabled=false}
   &CameraOffset(Camera Offset Amount=10)
   &DynamicMaterialParameters(
    Index 0 Param 1=1
    Index 0 Param 2=60
    Module.Param0WriteEnabled=true
    Module.Param1WriteEnabled=true
    Module.Param2WriteEnabled=true
    Module.Param3WriteEnabled=true
   )
  ]
  particleUpdateScript [
   &SkeletalMeshLocation(
    Apply Sampled Orientation to Mesh Particles=true
    Bone / Socket Index=6
    Inherit Velocity=false
    MaskSpawnGroup=false
    Mesh Sampling Type=::NewEnumerator5
    Sampled Position Offset=<32 0 0>
    Sampled Position Offset Coordinate Space:ENiagaraMeshTransforms=::NewEnumerator3
    Skeletal Mesh=$User.SkeletalMesh
    Socket Sampling Mode=::NewEnumerator5
    Use Sampled Mesh Attribute as Ribbon ID=false
    Write Sampled Mesh Attrributes to Particle Attributes=true
   )
   &ParticleState()
   /SetVariables(Particles.LightAlpha:NiagaraFloat=&FloatFromCurve(FloatCurve=DataInterface() {propertyValues {Curve {keys [
         {interpMode=RCIM_Cubic, tangentMode=RCTM_User, time=0.0271, value=-0.0101}
         {arriveTangent=0.0406, interpMode=RCIM_Cubic, leaveTangent=0.0406, tangentMode=RCTM_User, time=0.3777, value=0.984}
         {arriveTangent=-1.5812, interpMode=RCIM_Cubic, leaveTangent=-1.5812, tangentMode=RCTM_User, time=0.6521, value=0.8284}
         <at 1 0>
        ]}}})) {isSetParameters=true, module=SetVariables_#1}
   &Color(
    Color=&ColorFromCurve(ColorCurve=DataInterface() {propertyValues {AlphaCurve {keys [<at 0 0.6>]}, BlueCurve {keys [<at 0 0.3371>]}, GreenCurve {keys [<at 0 0.0249>]}, RedCurve {keys [<at 0 1>]}}})
    Scale Alpha=$Particles.LightAlpha
   )
   /NiagaraFN/Sprite/FN_SpriteAlignmentAndFacing(SpriteAlignment:Vector3f=<1 1 1>, SpriteFacing:Vector3f=<1 0 0>) {module=SpriteAlignmentAndFacing}
  ]
  rendererClasses [&NiagaraSpriteRendererProperties]
  renderers [{rendererClass=&NiagaraSpriteRendererProperties}]
 }
 NE_Decal001 {
  emitterUpdateScript [
   &EmitterState(Inactive Response=::NewEnumerator1, Life Cycle Mode=::NewEnumerator1, Loop Delay=0.6, Loop Duration=1.95, UseLoopDelay=true)
   &SpawnBurst_Instantaneous(Spawn Count=1, Spawn Time=0.05, Use Spawn Probability=false)
  ]
  particleSpawnScript [
   &InitializeParticle(
    Lifetime=2
    Module.Write Color=true
    Module.Write Lifetime=true
    Module.Write Mass=true
    Module.Write Material Random=true
    Module.Write Position=false
    Module.Write Scale=false
    Module.Write SpriteRotation=false
    Module.Write SpriteSize=true
    Sprite Size=&Vector2DFromFloat(Value=15)
   )
   &SystemLocation(MaskSpawnGroup=false, Offset=<20 0 10>) {enabled=false}
   &DynamicMaterialParameters(
    Index 0 Param 1=1
    Index 0 Param 2=60
    Module.Param0WriteEnabled=true
    Module.Param1WriteEnabled=true
    Module.Param2WriteEnabled=true
    Module.Param3WriteEnabled=true
   )
  ]
  particleUpdateScript [
   &SkeletalMeshLocation(
    Apply Sampled Orientation to Mesh Particles=true
    Bone / Socket Index=6
    Inherit Velocity=false
    MaskSpawnGroup=false
    Mesh Sampling Type=::NewEnumerator5
    Sampled Position Offset=<15 0 0>
    Sampled Position Offset Coordinate Space:ENiagaraMeshTransforms=::NewEnumerator3
    Skeletal Mesh=$User.SkeletalMesh
    Socket Sampling Mode=::NewEnumerator5
    Use Sampled Mesh Attribute as Ribbon ID=false
    Write Sampled Mesh Attrributes to Particle Attributes=true
   )
   &ParticleState()
   /SetVariables(Particles.LightAlpha:NiagaraFloat=&FloatFromCurve(FloatCurve=DataInterface() {propertyValues {Curve {keys [<at 0 0>,<at 0.1 1>,<at 0.65 1>,<at 1 0>]}}})) {isSetParameters=true, module=SetVariables_#1}
   &Color(
    Color=&ColorFromCurve(ColorCurve=DataInterface() {propertyValues {
       AlphaCurve {keys [<at -0.0046 1>]}
       BlueCurve {keys [<at -0.0017 0.7352>]}
       GreenCurve {keys [<at -0.0017 0.375>]}
       RedCurve {keys [<at -0.0017 1.5>]}
      }})
    Scale Alpha=$Particles.LightAlpha
   )
   &FN_SpriteAlignmentAndFacing(SpriteAlignment:Vector3f=<1 1 1>, SpriteFacing:Vector3f=<1 0 0>) {module=SpriteAlignmentAndFacing}
  ]
  rendererClasses [&NiagaraSpriteRendererProperties]
  renderers [{rendererClass=&NiagaraSpriteRendererProperties}]
 }
 WingFlap_MeshFoils {
  emitterSpawnScript [&FN_PreserveSkydiveUserVariables() {module=PreserveUserVariables}]
  emitterUpdateScript [
   &EmitterState()
   &SpawnRate(
    Spawn Probability=&MakeCustomFloatFromBool(Bool=$User.bIsFullyDeployed, True Float=1, False Float=0)
    SpawnRate=2
    Use Spawn Probability=true
   )
  ]
  particleSpawnScript [
   &FN_SetSpawnGroup_v2(Number of Spawn Groups=2)
   &InitializeParticle(
    Color=<rgba 1 1 1 1.15>
    Lifetime=2
    Mesh Scale=&MaskVectorBySpawnGroup(Masked Vector=<-1 1 1>, Passthrough Vector=<1 1 1>)
    Module.Write Color=true
    Module.Write Lifetime=true
    Module.Write Mass=false
    Module.Write Material Random=true
    Module.Write Position=false
    Module.Write Scale=true
    Module.Write SpriteRotation=false
    Module.Write SpriteSize=false
   )
   &CameraOffset(Camera Offset Amount=5)
  ]
  particleUpdateScript [
   &ParticleState()
   &SkeletalMeshLocation(
    Apply Sampled Orientation to Mesh Particles=true
    Bone / Socket Index=&MaskIntBySpawnGroup(Masked Int=4, Passthrough Int=5, Spawn Group Mask=0)
    Inherit Velocity=false
    MaskSpawnGroup=false
    Mesh Sampling Type=::NewEnumerator5
    Skeletal Mesh=$User.SkeletalMesh
    Socket Sampling Mode=::NewEnumerator5
    Use Sampled Mesh Attribute as Ribbon ID=false
    Write Sampled Mesh Attrributes to Particle Attributes=true
   )
   &ScaleColor(
    Scale Alpha=&FloatFromCurve(
     FloatCurve=DataInterface() {propertyValues {Curve {keys [<at 0 0 RCIM_Cubic>,<at 0.6 0.35 RCIM_Cubic>,<at 1 0 RCIM_Cubic>]}}}
     Scale Curve=$System.FN_Integrator_v2.Attr
    )
    ScaleA=true
    ScaleRGB=false
   )
  ]
  rendererClasses [&NiagaraMeshRendererProperties]
  renderers [{rendererClass=&NiagaraMeshRendererProperties, rendererIndex=0}]
 }
}
summary {
 systemName=NS_Glider_Deli_Sandwich_Trails_01
 emitters [
  {emitterName=DeliSandwish_VFX, rendererClasses [&NiagaraMeshRendererProperties]}
  {emitterName=NE_Decal, rendererClasses [&NiagaraSpriteRendererProperties]}
  {emitterName=NE_Decal001, rendererClasses [&NiagaraSpriteRendererProperties]}
  {emitterName=WingFlap_MeshFoils, rendererClasses [&NiagaraMeshRendererProperties]}
 ]
 userVariables [
  User.bIsFrontEndPreview {type=NiagaraBool}
  User.bIsFullyDeployed {type=NiagaraBool}
  User.ForwardDot {type=NiagaraFloat, defaultValue=0, description=""}
  User.LocalAlpha {type=NiagaraFloat, defaultValue=1, description=""}
  User.PlayerSpeed {type=NiagaraFloat, defaultValue=0, description=""}
  User.RibbonIDOffset {type=NiagaraInt32}
  User.RightDot {type=NiagaraFloat, defaultValue=0, description=""}
  User.SkeletalMesh {
   type=NiagaraDataInterfaceSkeletalMesh
   defaultValue {
    struct=NiagaraExt_VariableValue_DataInterface
    value {dataInterface=&self:NiagaraDataInterfaceSkeletalMesh_0}
   }
  }
  User.TimeOfDay {type=NiagaraFloat, defaultValue=0, description=""}
 ]
}
